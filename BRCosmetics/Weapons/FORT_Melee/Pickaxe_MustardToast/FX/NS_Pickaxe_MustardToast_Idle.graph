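# Niagara idle effect for the MustardToast pickaxe, layering a persistent red glow with drifting heat particles.
# CapsuleGlow bursts an aligned, camera-offset sprite, while SplineThickenGlow bursts meshes through two renderers, scaled by User.FXScale, with dynamic material parameters and a boosted red ColorGlow attribute.
# HeatParticles spawns SubUV sprites in a thin cylinder that accelerate upward, dim and fade, with spawn probability dropping to zero while User.bisSwinging is true.
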NiagaraSystem /BRCosmetics/Weapons/FORT_Melee/Pickaxe_MustardToast/FX/NS_Pickaxe_MustardToast_Idle
dependencies {
 usedDataInterfaces [
  &NiagaraDataInterfaceSpriteRendererInfo
  &NiagaraDataInterfaceCurve
  &NiagaraDataInterfaceColorCurve
 ]
 usedDynamicInputs [&ColorFromCurve, &FNTimeOfDayScalar]
 usedModules [
  &SystemState
  &self:SystemSpawnScript.SystemScriptSource.SystemScriptGraph.NiagaraNodeAssignment_1.SetVariables_#1
  &self:SystemSpawnScript.SystemScriptSource.SystemScriptGraph.NiagaraNodeAssignment_2.SetVariables_#2
 ]
}
emitters {
 CapsuleGlow {
  emitterUpdateScript [
   &EmitterState(Life Cycle Mode=::NewEnumerator1, Loop Behavior=::NewEnumerator1, Loop Duration Mode=::NewEnumerator1, UseLoopDelay=false)
   &SpawnBurst_Instantaneous(Use Spawn Probability=false, UseLoopCountLimit=true)
   /Niagara/Noise/Waveform(
    EnableExponent1=false
    Global Input Position:NiagaraFloat=&RandomRangeFloat(Minimum=0, Maximum=0.1)
    GlobalClampMax=false
    GlobalClampMin=false
    UseInput1=false
    UsePhase1=false
    "[1] Frequency":NiagaraFloat=&RandomRangeFloat(Minimum=0.15, Maximum=0.25)
   ) {enabled=false, module=Waveform001}
  ]
  particleSpawnScript [
   &InitializeParticle(
    Color=<rgba 1 0.0462 0.03 0.2>
    Mesh Scale Mode=::NewEnumerator3
    Mesh Uniform Scale=&Multiply_Float(A=1, B=$User.FXScale)
    Position Mode=::NewEnumerator0
    Sprite Rotation Mode=::NewEnumerator0
    Sprite Size=&Multiply_Vector2D_ByFloat(Float=1, Vector2D=<50 127>)
    Sprite Size Mode=::NewEnumerator4
    Write Material Random=true
   )
   &SpriteFacingAndAlignment(Sprite Alignment=<0.078 0 1>, Sprite Facing=<0.7068 0.07 0.3>, WriteAlignment=true, WriteFacing=false)
   &SystemLocation(MaskSpawnGroup=false, Offset=<0.6 0 58.5>)
   &CameraOffset(Camera Offset Amount=-10)
  ]
  particleUpdateScript [
   &ParticleState(Kill Particles When Lifetime Has Elapsed=false, Let Infinitely Lived Particles Die When Emitter Deactivates=true)
   &DynamicMaterialParameters(
    Index 0 Param 1=5
    Module.Index0Param0Write=true
    Module.Index0Param1Write=false
    Module.Index0Param2Write=false
    Module.Index0Param3Write=false
   )
  ]
  rendererClasses [&NiagaraSpriteRendererProperties]
  renderers [{rendererClass=&NiagaraSpriteRendererProperties}]
 }
 HeatParticles {
  emitterUpdateScript [
   &EmitterState(Life Cycle Mode=::NewEnumerator1, Loop Duration=2, UseLoopDelay=false)
   &SpawnRate(
    Spawn Probability=&MakeCustomFloatFromBool(Bool=$User.bisSwinging, True Float=0, False Float=1)
    SpawnRate=10
    Use Spawn Probability=true
   )
  ]
  particleSpawnScript [
   &InitializeParticle(
    Color=<rgba 1 0.2137 0.1539 1>
    Lifetime Max=1.5
    Lifetime Min=0.8
    Lifetime Mode=::NewEnumerator1
    Mass Mode=::NewEnumerator3
    Position Offset=<0 0 50>
    Sprite Rotation Mode=::NewEnumerator2
    Sprite Size Mode=::NewEnumerator1
    Uniform Sprite Size Max=38
    Uniform Sprite Size Min=25
    UsePositionOffset=false
    Write Material Random=true
   )
   &ShapeLocation(
    Cylinder Height=70
    Cylinder Radius=2
    MaskSpawnGroup=false
    Offset=<0 0 61>
    Offset Coordinate Space=::Local
    Offset Mode=::NewEnumerator0
    Override Seed=false
    Shape Primitive=::NewEnumerator1
    Surface Only=false
    UseOwnerScale=true
   )
   &SubUVAnimation(
    Evaluation Type=::NewEnumerator0
    "Sprite Renderer "=DataInterface() {propertyValues {SpriteRenderer=&self:HeatParticles_2.NiagaraSpriteRendererProperties_0}}
    SubUV Animation Mode:ENiagara_SubUVLookupModeV2=::NewEnumerator2
    UseEndFrame=false
    UseStartFrame=false
   )
  ]
  particleUpdateScript [
   &ParticleState()
   &ScaleSpriteSize()
   &ScaleColor(
    Scale Alpha=&FloatFromCurve(
     FloatCurve=DataInterface() {propertyValues {Curve {keys [<at 0 0 RCIM_Cubic>,<at 0.4 1 RCIM_Cubic>,<at 1 0 RCIM_Cubic>]}}}
     Scale Curve=&RandomRangeFloat(Minimum=0.05, Maximum=0.1)
    )
    Scale RGB=&VectorFromFloat(Value=&Lerp_Float(
      A=3
      B=0.1
      Alpha=&FloatFromCurve(FloatCurve=DataInterface() {propertyValues {Curve {keys [<at 0 0>,<at 1 1>]}}})
     ))
    ScaleA=true
    ScaleRGB=true
   )
   &AccelerationForce(Acceleration=<0 0 4>)
   &SolveForcesAndVelocity(Clamp Velocity=false, Limit Acceleration=false)
  ]
  rendererClasses [&NiagaraSpriteRendererProperties]
  renderers [{rendererClass=&NiagaraSpriteRendererProperties}]
 }
 SplineThickenGlow {
  emitterUpdateScript [
   &EmitterState(Life Cycle Mode=::NewEnumerator1, Loop Behavior=::NewEnumerator1, Loop Duration Mode=::NewEnumerator1, UseLoopDelay=false)
   &SpawnBurst_Instantaneous(Use Spawn Probability=false, UseLoopCountLimit=true)
   &Waveform(
    EnableExponent1=false
    Global Input Position:NiagaraFloat=&RandomRangeFloat(Minimum=0, Maximum=0.1)
    GlobalClampMax=false
    GlobalClampMin=false
    UseInput1=false
    UsePhase1=false
    "[1] Frequency":NiagaraFloat=&RandomRangeFloat(Minimum=0.15, Maximum=0.25)
   ) {enabled=false, module=Waveform001}
  ]
  particleSpawnScript [
   &InitializeParticle(
    Color=<rgba 20 18.1333 18 0.35>
    Mesh Scale=&Multiply_VectorByFloat(Vector=<1 1 1.01>, Float=&Multiply_Float(A=1, B=$User.FXScale))
    Mesh Scale Mode=::NewEnumerator4
    Position Mode=::NewEnumerator0
    Sprite Rotation Mode=::NewEnumerator0
    Sprite Size=&Multiply_Vector2D_ByFloat(Float=1, Vector2D=<35 115>)
    Sprite Size Mode=::NewEnumerator4
    Write Material Random=true
   )
   &SpriteFacingAndAlignment(Sprite Alignment=<0.084 0 1>, Sprite Facing=<0.7068 0.07 0.3>, WriteAlignment=true, WriteFacing=false)
  ]
  particleUpdateScript [
   &ParticleState(Kill Particles When Lifetime Has Elapsed=false, Let Infinitely Lived Particles Die When Emitter Deactivates=true)
   &DynamicMaterialParameters(
    Index 0 Param 2=1
    Index 0 Param 4=1
    Index 1 Param 1=12702.6572
    Index 1 Param 2=115.411
    Index 1 Param 3=50.9432
    Index 1 Param 4=24.9799
    Index 2 Param 1=-2
    Index 2 Param 2=2
    Index 2 Param 3=1
    Index 3 Param 3=0.4
    Index 3 Param 4=0.8
    Module.Index0Param0Write=true
    Module.Index0Param1Write=true
    Module.Index0Param2Write=false
    Module.Index0Param3Write=false
    Module.Index1Param0Write=false
    Module.Index1Param1Write=false
    Module.Index1Param2Write=false
    Module.Index1Param3Write=false
    Module.Index2Param0Write=true
    Module.Index2Param1Write=true
    Module.Index2Param2Write=true
    Module.Index2Param3Write=false
    Module.Index3Param0Write=true
    Module.Index3Param1Write=false
    Module.Index3Param2Write=true
    Module.Index3Param3Write=true
    Write Parameter Index 1=true
    Write Parameter Index 2=true
    Write Parameter Index 3=true
   )
   /SetVariables(Particles.ColorGlow:LinearColor=&MultiplyLinearColor(
     Linear Color=$Particles.Color
     Scale Factor=&MakeVector4_FromLinearColor(LinearColor=<rgba 100 6.2333 3 1>)
    )) {isSetParameters=true}
  ]
  rendererClasses [&NiagaraMeshRendererProperties]
  renderers [
   {rendererClass=&NiagaraMeshRendererProperties, rendererIndex=0}
   {rendererClass=&NiagaraMeshRendererProperties, rendererIndex=1}
  ]
 }
}
summary {
 systemName=NS_Pickaxe_MustardToast_Idle
 emitters [
  {emitterName=CapsuleGlow, rendererClasses [&NiagaraSpriteRendererProperties]}
  {emitterName=HeatParticles, rendererClasses [&NiagaraSpriteRendererProperties]}
  {emitterName=SplineThickenGlow, rendererClasses [&NiagaraMeshRendererProperties]}
 ]
 userVariables [
  User.bisSwinging {type=NiagaraBool}
  User.FXScale {type=NiagaraFloat, defaultValue=1, description=""}
  User.NewVector3f {type=Vector3f, description="", defaultValue {struct=Vector3f, value=<0 0 0>}}
  User.RibbonIDOffset {type=NiagaraInt32}
  User.Variant {type=NiagaraFloat, defaultValue=0, description=""}
 ]
}
