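# Physics asset for the MustardToast axe mesh, a two-handed melee pickaxe cosmetic.
# It holds a single capsule body on the Melee_TwoHanded_Handle bone, offset from the bone origin and rotated.
# It has no constraints, so the whole weapon is represented by that one simple collision shape with no jointed parts.
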PhysicsAsset /BRCosmetics/Weapons/FORT_Melee/Pickaxe_MustardToast/Meshes/MustardToast_Axe_PhysicsAsset
bodies {Melee_TwoHanded_Handle [{shapeType=Capsule, length=61.5247, radius=6.139, center=<-1.2817 -8.1477 69.5798>, rotation=<rot 0 -90 1.9197>}]}
constraints []
